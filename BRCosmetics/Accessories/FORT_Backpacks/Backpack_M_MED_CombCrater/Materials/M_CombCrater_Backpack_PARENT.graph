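# Parent material for the CombCrater backpack, adding a timed glowing symbol effect between the 2023 character PreFX and PostFX functions.
# Time since BP_GetTime builds a fade-in, hold and fade-out value from duration and speed parameters.
# Noise thresholding and panning-texture UV distortion dissolve the symbol texture.
# Emissive combines symbol masks and a pre-skinned-position sphere glow, tinted by time of day, while base color blends toward FX_BaseColor.
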Material /BRCosmetics/Accessories/FORT_Backpacks/Backpack_M_MED_CombCrater/Materials/M_CombCrater_Backpack_PARENT
graph {
 properties {
  bUseMaterialAttributes=true
  bUsedWithMeshDeformer=true
  bUsedWithMorphTargets=true
  bUsedWithNiagaraMeshParticles=true
  bUsedWithSkeletalMesh=true
  pixelDepthOffsetMode=PDOM_Legacy
  shadingModel=MSM_FromMaterialExpression
  usedShadingModels=Default Lit | Subsurface | Clear Coat | Subsurface Profile | Cloth
 }
 nodes [
  n1 = MaterialFunctionCall(Override Tech Art Mask UVs=NamedRerouteUsage()) {materialFunction=&MF_CharacterParent_2023_PreFX}
  n2 = GetMaterialAttributes(None=@n1.Result) {attributeGetTypes [#1]}
  MaterialFunctionCall(
   Material=SetMaterialAttributes!(MaterialAttributes=@n2.MaterialAttributes, Emissive Color=NamedRerouteUsage(), Base Color=NamedRerouteUsage()) {attributeSetTypes [#2, #1]}
   M=@n1.M
   Pre Skinned Local Position=@n1.M
   Distance Blend=@n1.M
   Fresnel=@n1.M
  ) {materialFunction=&MF_CharacterParent_2023_PostFX}
  n3 = Time()
  n4 = Subtract(A=@n3, B=ScalarParameter() {defaultValue=999999986991104, parameterName=BP_GetTime})
  n5 = Saturate(None=Multiply(A=Subtract(A=@n4) {constB=-0.32}, B=ScalarParameter() {defaultValue=0.7, parameterName=FX_Speed_FadeIn}) {constB=0.5})
  n6 = ScalarParameter() {defaultValue=0.38, parameterName=FX_Speed}
  n7 = Multiply(A=Subtract(A=@n4, B=ScalarParameter() {defaultValue=5.5, parameterName=FX_Duration}) {constB=5}, B=@n6) {constB=0.5}
  NamedRerouteDeclaration(None=Add(A=OneMinus(None=@n5), B=Saturate(None=@n7))) {name=FadingTiming, nodeColor=<rgba 0.156 0.1691 1 1>}
  n8 = Multiply(A=ComponentMask(None=NamedRerouteUsage()) {g=true}) {constB=1.5}
  n9 = TextureCoordinate()
  n10 = MaterialFunctionCall(UV Channel=Add(A=OneMinus(None=@n9)) {constB=-1}) {materialFunction=&LinearGradient}
  n11 = NamedRerouteUsage()
  n12 = OneMinus(None=SmoothStep(Value=@n11) {constMax=0.7, constMin=0.05})
  n13 = Multiply(A=Saturate(None=Add(A=@n10.VGradient, B=@n12)), B=NamedRerouteUsage())
  n14 = NamedRerouteUsage()
  n15 = SmoothStep(Value=@n11) {constMax=0.5, constMin=0.07}
  n16 = NamedRerouteDeclaration(None=Add(A=Multiply(A=@n8, B=@n13), B=Multiply(A=@n14, B=@n15))) {name="SymbolBlack&&White", nodeColor=<rgba 1 0.9609 0 1>}
  n17 = VectorParameter() {parameterName=FX_GlowColor_Day, defaultValue=<rgba 0.1289 0 1 1>}
  n18 = VectorParameter() {parameterName=FX_GlowColor_Night, defaultValue=<rgba 0.1376 0 0.7656 1>}
  n19 = MaterialFunctionCall(Color_Morning=@n17.RGB, Color_Day=@n17.RGB, Color_Evening=@n18.RGB, Color_Night=@n18.RGB) {materialFunction=&Color_Time_of_Day}
  n20 = Multiply(B=@n19.Result) {constA=8}
  n21 = SphereMask(A=NamedRerouteUsage(), B=Constant3Vector() {constant=<rgba 0 -13.8 -8 0>}) {attenuationRadius=20, hardnessPercent=0}
  n22 = Multiply(A=Multiply(A=@n19.Result, B=Power(Base=@n21) {constExponent=4})) {constB=1.5}
  n23 = NamedRerouteUsage()
  n24 = LinearInterpolate(Alpha=SmoothStep(Value=Multiply(A=@n23, B=@n23)) {constMin=0.2}) {constA=0.16, constB=0}
  NamedRerouteDeclaration(None=Add(A=Multiply(A=@n16, B=@n20), B=Multiply(A=@n22, B=@n24))) {name=EmissiveOutput, nodeColor=<rgba 0.1719 1 0 1>}
  n25 = TextureSample(UVs=TextureCoordinate() {uTiling=0.5, vTiling=0.5}) {samplerSource=SSM_Wrap_WorldGroupSettings, texture=/BRCosmetics/Accessories/FORT_Backpacks/Backpack_M_MED_CombCrater/Textures/3d_VecNoise}
  n26 = NamedRerouteUsage()
  n27 = Saturate(None=Add(A=@n25.B, B=LinearInterpolate(Alpha=@n26) {constA=1, constB=-1}))
  n28 = TextureSample(UVs=Panner(Coordinate=TextureCoordinate() {uTiling=3, vTiling=3}) {speedX=-0.07, speedY=0.35}) {
   samplerType=SAMPLERTYPE_Normal
   texture=/BRCosmetics/Accessories/FORT_Backpacks/Backpack_M_MED_CombCrater/Textures/T_CombCrater_Backpack_DissolveTexture
  }
  n29 = ComponentMask(None=@n28.RGB) {g=true, r=true}
  n30 = Constant2Vector() {g=1, r=0.3}
  n31 = Add(A=TextureCoordinate(), B=Multiply(A=LinearInterpolate(Alpha=@n26) {constB=0.22}, B=Multiply(A=@n29, B=@n30))) {constB=0}
  n32 = LinearInterpolate(B=Constant2Vector() {g=1.1, r=1}, Alpha=@n26) {constA=1, constB=1.5}
  n33 = TextureSample!(
   UVs=NamedRerouteDeclaration(None=Multiply(A=@n31, B=@n32)) {name=SymbolDistortion, nodeColor=<rgba 0.4531 0 1 1>}
   Tex
   Bias=Constant() {r=3}
   Apply View MipBias
  ) {
   mipValueMode=TMVM_MipBias
   samplerSource=SSM_Wrap_WorldGroupSettings
   texture=/BRCosmetics/Accessories/FORT_Backpacks/Backpack_M_MED_CombCrater/Textures/T_CombCrater_Backpack_FX
  }
  NamedRerouteDeclaration(None=Multiply(A=NamedRerouteDeclaration(None=@n27) {name=FadingMask, nodeColor=<rgba 1 0.9463 0.9099 1>}, B=@n33.G)) {name=FadingAnimation_Texture, nodeColor=<rgba 1 0.3047 0 1>}
  NamedRerouteDeclaration(None=@n1.Pre skinned Local Position) {name=PreskinnedPosition, nodeColor=<rgba 1 0 0.6562 1>}
  NamedRerouteDeclaration(None=@n1.Tech Art Mask) {name=TA_Mask, nodeColor=<rgba 1 0 0.4922 1>}
  NamedRerouteDeclaration(None=@n2.Base Color) {name=OriginalBaseColor, nodeColor=<rgba 1 0.2186 0.0061 1>}
  n34 = VectorParameter() {parameterName=FX_BaseColor, defaultValue=<rgba 0.4019 0.0201 1 1>}
  NamedRerouteDeclaration(None=LinearInterpolate(A=NamedRerouteUsage(), B=Multiply(A=@n34.RGB) {constB=2}, Alpha=NamedRerouteUsage())) {name=BaseColorOutput, nodeColor=<rgba 0 1 0.3203 1>}
  Reroute(None=@n27)
  Reroute(None=Frac(None=Multiply(A=@n3, B=@n6) {constB=0.4}))
 ]
}
