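# Animation Blueprint for the CombCrater backpack cosmetic, playing an idle pose and an elimination reaction.
# Output blends by bIsInFrontEnd between a front-end machine and a gameplay machine.
# The backpack part modifier's FE_Reactive flag starts delayed front-end reactions, repeated via notify, while an enemy-eliminated input event triggers gameplay ones.
# Parachute alphas and sine float are computed but unread here, and the pack bone-offset chain is disconnected from output.
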AnimBlueprint /BRCosmetics/Accessories/FORT_Backpacks/Backpack_M_MED_CombCrater/Meshes/M_MED_CombCrater_Pack_AnimBP
target_skeleton=&/M_MED_CombCrater_Pack_Skeleton
graphs [
 AnimGraph {nodes [
   Input Pose()
   n1 = "Transform (Modify) Bone - Bone: pack"(Translation="0,-2.8,0", Rotation="0,0,0") {
    translationMode=BMM_Additive
    translationSpace=BCS_BoneSpace
    alphaBoolBlend {blendOption=Linear}
    boneToModify {boneName=pack}
    translation {y=-2.8}
   }
   n2 = "Transform (Modify) Bone - Bone: pack"(ComponentPose=@n1, Translation="0,0,0", Rotation="0,0,0") {translationMode=BMM_Additive, translationSpace=BCS_BoneSpace, alphaBoolBlend {blendOption=Linear}, boneToModify {boneName=pack}}
   n3 = "Transform (Modify) Bone - Bone: pack"(ComponentPose=@n2, Translation="0,0,0", Rotation="0,0,25") {
    rotationMode=BMM_Additive
    rotationSpace=BCS_BoneSpace
    translationMode=BMM_Additive
    translationSpace=BCS_BoneSpace
    alphaBoolBlend {blendOption=Linear}
    boneToModify {boneName=pack}
    rotation {roll=25}
   }
   "Transform (Modify) Bone - Bone: pack"(ComponentPose=@n3, Translation="0,0,0", Rotation="0,0,25") {
    rotationMode=BMM_Additive
    rotationSpace=BCS_BoneSpace
    translationMode=BMM_Additive
    translationSpace=BCS_BoneSpace
    alphaBoolBlend {blendOption=Linear}
    boneToModify {boneName=pack}
    rotation {roll=25}
   }
   AnimGraphNode_StateMachine_1 = FE_Reactivity() {is=title AnimGraphNode_StateMachine}
   AnimGraphNode_StateMachine_2 = Backpack States() {is=title AnimGraphNode_StateMachine}
   Output Pose(Result=Blend Poses by bool(BlendPose_0=@AnimGraphNode_StateMachine_1, BlendPose_1=@AnimGraphNode_StateMachine_2, bActiveValue=Get bIsInFrontEnd()) {blendType=Linear}) {layerGroup=None}
  ]}
 FE_Reactivity {
  entry=Idle
  path=AnimGraph.AnimGraphNode_StateMachine_1.FE_Reactivity
  ids=id
  nodes [
   AnimStateNode_0 = Idle()
   AnimStateNode_2 = FE_Elim() {is=title AnimStateNode}
   AnimStateTransitionNode_8 = AnimStateTransitionNode()
   AnimStateTransitionNode_9 = AnimStateTransitionNode() {bAutomaticRuleBasedOnSequencePlayerInState=true}
  ]
  links [
   {
    dst=Idle
    rule=AnimGraph.AnimGraphNode_StateMachine_1.FE_Reactivity.AnimStateTransitionNode_9.Transition
    src=FE_Elim
    via=AnimStateTransitionNode_9
   }
   {
    dst=FE_Elim
    rule=AnimGraph.AnimGraphNode_StateMachine_1.FE_Reactivity.AnimStateTransitionNode_8.Transition
    src=Idle
    via=AnimStateTransitionNode_8
   }
  ]
 }
 Idle {
  path=AnimGraph.AnimGraphNode_StateMachine_1.FE_Reactivity.AnimStateNode_0.Idle
  nodes [Output Animation Pose(Result="Sequence Evaluator (by time) 'CombCrater_Idle_v1'"!(ExplicitTime:float=0) -> (Pose:Pose Link Structure) {
     is=title AnimGraphNode_SequenceEvaluator
     sequence=/BRCosmetics/Accessories/FORT_Backpacks/Backpack_M_MED_CombCrater/Animation/CombCrater_Idle_v1
    })]
 }
 FE_Elim {
  path=AnimGraph.AnimGraphNode_StateMachine_1.FE_Reactivity.AnimStateNode_2.FE_Elim
  nodes [Output Animation Pose(Result=Sequence Player 'CombCrater_Elim'() {bLoopAnimation=false, sequence=/BRCosmetics/Accessories/FORT_Backpacks/Backpack_M_MED_CombCrater/Animation/CombCrater_Elim})]
 }
 Transition {
  path=AnimGraph.AnimGraphNode_StateMachine_1.FE_Reactivity.AnimStateTransitionNode_8.Transition
  nodes [Result(bCanEnterTransition=Get PlayFE_Elim() -> (PlayFE_Elim:bool))]
 }
 Transition {path=AnimGraph.AnimGraphNode_StateMachine_1.FE_Reactivity.AnimStateTransitionNode_9.Transition, nodes [Result()]}
 Backpack States {
  entry=Idle_pose
  path=AnimGraph.AnimGraphNode_StateMachine_2.Backpack States
  ids=id
  nodes [
   AnimStateNode_0 = Idle_pose() {
    is=title AnimStateNode
    stateEntered {guid=#1, notifyName=Entered_RandomLocoBreak}
    stateFullyBlended {guid=#2}
    stateLeft {guid=#3}
   }
   AnimStateNode_15 = Elimination() {
    is=title AnimStateNode
    stateEntered {guid=#4}
    stateFullyBlended {guid=#5}
    stateLeft {guid=#6}
   }
   AnimStateTransitionNode_16 = AnimStateTransitionNode() {
    transitionEnd {guid=#7}
    transitionInterrupt {guid=#8}
    transitionStart {guid=#9}
   }
   AnimStateTransitionNode_27 = AnimStateTransitionNode() {bAutomaticRuleBasedOnSequencePlayerInState=true}
  ]
  links [
   {
    dst=Idle_pose
    rule=AnimGraph.AnimGraphNode_StateMachine_2.Backpack States.AnimStateTransitionNode_27.Transition
    src=Elimination
    via=AnimStateTransitionNode_27
   }
   {
    dst=Elimination
    rule=AnimGraph.AnimGraphNode_StateMachine_2.Backpack States.AnimStateTransitionNode_16.Transition
    src=Idle_pose
    via=AnimStateTransitionNode_16
   }
  ]
 }
 Idle_pose {
  path=AnimGraph.AnimGraphNode_StateMachine_2.Backpack States.AnimStateNode_0.Idle_pose
  nodes [Output Animation Pose(Result=Sequence Player 'CombCrater_Idle_v1'() {sequence=&CombCrater_Idle_v1})]
 }
 Elimination {
  path=AnimGraph.AnimGraphNode_StateMachine_2.Backpack States.AnimStateNode_15.Elimination
  nodes [Output Animation Pose(Result=Sequence Player 'CombCrater_Elim'() {bLoopAnimation=false, sequence=&CombCrater_Elim})]
 }
 Transition {
  path=AnimGraph.AnimGraphNode_StateMachine_2.Backpack States.AnimStateTransitionNode_16.Transition
  nodes [Result(bCanEnterTransition=Get bEnemyPlayerEliminated() -> (bEnemyPlayerEliminated:bool))]
 }
 Transition {path=AnimGraph.AnimGraphNode_StateMachine_2.Backpack States.AnimStateTransitionNode_27.Transition, nodes [Result()]}
 EventGraph {nodes [
   Event BlueprintUpdateAnimation()
   > n1 = Sequence()
   n2 = GetRealTimeSeconds()
   n3 = "Sin (Radians)"(A=float * float(A=float * float(A=1.11, B=1.11), B=@n2))
   n4 = "Sin (Radians)"(A=float * float(A=float * float(A=1.24, B=0.8), B=@n2))
   n5 = "Sin (Radians)"(A=float * float(A=float * float(A=1.8, B=0.5), B=@n2))
   Set dyn_pack_Float(
    execute=@n1.then_1
    dyn_pack_Float:Vector=MakeVector(X=float * float(A=@n4, B=2.3), Y=float * float(A=@n3, B=0.3), Z=float * float(A=@n5, B=2))
   ) -> (Output_Get:Vector)
   n6 = GetOwnerSkeletalMeshComponent()
   Set Alpha parachute opening(
    execute=@n1.then_0
    Alpha parachute opening:float=GetAlphaFromReferenceTransformDelta(SkeletalMeshComponent=@n6, BoneName=spine_05, BaseName=pelvis, InputRange=(X=-35,Y=-90), TwistAxis="1,0,0")
   ) -> (Output_Get:float)
   > Set Alpha parachute hide(Alpha parachute hide:float=GetAlphaFromReferenceTransformDelta(SkeletalMeshComponent=@n6, BoneName=spine_01, BaseName=pelvis, InputRange=(X=3,Y=20), TwistAxis="1,0,0")) -> (Output_Get:float)
   Event BlueprintPostEvaluateAnimation() {is=title K2Node_Event}
   > Set bEnemyPlayerEliminated(bEnemyPlayerEliminated:bool=false) -> (Output_Get:bool)
   Event OnCharacterCustomizationLoaded()
   > n7 = Cast To FortPlayerPawn(Object=TryGetFortPawn())
   > n8 = Cast To B_Athena_Part_Modifier_Backpack_CombCrater!(
    execute
    Object=GetModifierForPartType!(self:Fort Player Pawn Object Reference=@n7, PartType:EFortCustomPartType Enum=Backpack) -> (ReturnValue:Custom Character Part Modifier Object Reference)
   ) -> (then, CastFailed, AsB Athena Part Modifier Backpack Comb Crater:B Athena Part Modifier Backpack Comb Crater Object Reference) {is=title K2Node_DynamicCast}
   > n9 = Branch(Condition=Get FE_Reactive!(self:B Athena Part Modifier Backpack Comb Crater Object Reference=@n8) -> (FE_Reactive:bool))
   > Do Once(Reset=@n11.then)
   > n10 = Delay(Duration=4)
   Set PlayFE_Elim(execute=[@n10.then, @n12.then], PlayFE_Elim:bool=true) -> (Output_Get:bool)
   n11 = Set PlayFE_Elim(execute=@n9.else, PlayFE_Elim=false)
   AnimNotify_ReactiveReset() {is=title K2Node_Event}
   > Set PlayFE_Elim(PlayFE_Elim=false)
   > n12 = Delay(Duration=10)
   n13 = Event BlueprintProcessAnimInputEvent()
   > Branch(Condition=IsNamed(self=@n13.AnimInputEvent, InEventName=aie.EnemyPlayerEliminated))
   > Set bEnemyPlayerEliminated(bEnemyPlayerEliminated=true)
   > RetriggerableDelay(Duration=1.3)
   > Set bEnemyPlayerEliminated(bEnemyPlayerEliminated=false)
   > Set isCombat(isCombat:bool=false) -> (Output_Get:bool)
   TryGetPawnOwner()
  ]}
]
variables [
 Alpha parachute hide {type=number}
 Alpha parachute opening {type=number}
 bEnemyPlayerEliminated {type=boolean}
 dyn_pack_Float {type=Vector, default=<-1.52 0.1398 -1.9741>}
 isCombat {type=boolean}
 PlayFE_Elim {type=boolean}
]
